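; A_BasilOkra_Instructions_Spawner is an actor that spawns the BasilOkra instruction manual in the WildEstate_BasilOkra content.
; Its SpawnInstructionManual function spawns a BGA_BasilOkra_Instructions actor at the spawner's own transform, adjusting position if possible but always spawning, and returns the new actor.
; The construction script is empty, and nothing in the file calls the spawn function.
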(fn SpawnInstructionManual
  (bind _returnvalue (Game|SpawnActorfromClass "/WildEstate_BasilOkra/Gameplay/Actors/BGA_BasilOkra_Instructions.BGA_BasilOkra_Instructions_C" (Transformation|GetActorTransform self) "AdjustIfPossibleButAlwaysSpawn" "MultiplyWithRoot" 0 0 0))
  (return _returnvalue))

(fn ConstructionScript)
