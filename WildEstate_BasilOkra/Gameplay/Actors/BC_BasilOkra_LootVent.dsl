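; Loot vent actor that, once switched on, launches overlapping pawns along its up vector and applies fall-immunity, redeploy and temporary build-destruction effects.
; TurnCollisionOn and TurnCollisionOff toggle launch collision, VFX and audio, plus an authority-only looping timer.
; That timer runs a targeting preset to destroy buildings at or below a health threshold.
; Custom events spawn a marker-pinged slingshot pickup or Slurp pickups scaled by playlist tags.
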(event EventBeginPlay
  (Collision|SetCollisionEnabled (Variables|Default|GetLaunchCollision) "NoCollision"))

(event Collision|EventActorBeginOverlap
  (params (OtherActor :type Actor%20Object%20Reference)))

(event EventTick
  (params (DeltaSeconds :type Float%20%28single-precision%29)))

(event OnComponentBeginOverlap(LaunchCollision)
  (params (OverlappedComponent :type Primitive%20Component%20Object%20Reference) (OtherActor :type Actor%20Object%20Reference) (OtherComp :type Primitive%20Component%20Object%20Reference) (OtherBodyIndex :type Integer) (bFromSweep :type Boolean) (SweepResult :type Hit%20Result%20Structure%20%28by%20ref%29))
  (bind _asfort_pawn (Utilities|Casting|CastToFortPawn OtherActor))
  (Character|LaunchCharacter _asfort_pawn (* (Transformation|GetActorUpVector) (ScalableFloat|GetValueatLevel (Variables|Config|GetLaunchVelocity))) (ScalableFloat|GetValueasBool (Variables|Config|GetShouldXYOverride)) (ScalableFloat|GetValueasBool (Variables|Config|GetShouldZOverride)))
  (Environment|ApplyGlobalEnvironmentGameplayEffecttoActor _asfort_pawn (Variables|Default|GetFall%20Damage%20Immunity%20Effect%20Class))
  (Environment|ApplyGlobalEnvironmentGameplayEffecttoActor _asfort_pawn (Variables|Default|GetRedeploy%20Effect%20Class))
  (bind _returnvalue (Environment|ApplyGlobalEnvironmentGameplayEffecttoActor _asfort_pawn (Variables|Default|GetDestroyBuildsEffectClass)))
  (Variables|Default|SetDestroyBuildsGEHandle _returnvalue)
  (Audio|SpawnSoundAttached (Variables|Sound|GetPlayerLaunchedByVentSound) (Variables|/Script/Engine.Character|GetMesh _asfort_pawn)))

(event OnComponentEndOverlap(LaunchCollision)
  (params (OverlappedComponent :type Primitive%20Component%20Object%20Reference) (OtherActor :type Actor%20Object%20Reference) (OtherComp :type Primitive%20Component%20Object%20Reference) (OtherBodyIndex :type Integer))
  (bind _asfort_pawn (Utilities|Casting|CastToFortPawn OtherActor))
  (GameplayEffects|RemoveActiveGameplayEffect (Ability|GetAbilitySystemComponent _asfort_pawn) (Variables|Default|GetDestroyBuildsGEHandle) -1))

(event Custom|CheckForBuilds
  (bind _returnvalue (Utilities|Time|SetTimerbyEvent (delegate Custom|BasilOkra_CheckForBuilds_Loop) (ScalableFloat|GetValueatLevel (Variables|Config|GetCheckBuildTickRate)) true))
  (Variables|Default|SetCheckBuildTimerHandle _returnvalue))

(event Custom|BasilOkra_CheckForBuilds_Loop
  (bind (_asynctaskproxy _targetinghandle) (Class|AsyncActionPerformTargeting|PerformTargetingAsyncAction 0 "/WildEstate_BasilOkra/Gameplay/TP_BasilOkra_Vent_DestroyBuilds.TP_BasilOkra_Vent_DestroyBuilds")
    (:Targeted
      (for _array_element (TargetingSystem|TargetingResults|GetTargetingResultsActors (GameInstanceSubsystems|GetTargetingSubsystem) _targetinghandle)
        (bind _asbuilding_actor (Utilities|Casting|CastToBuildingActor _array_element))
        (if (<= (Health|GetHealth _asbuilding_actor) (Variables|Config|GetMaxHealthToAutoDestroy))
          (Actor|DestroyActor _asbuilding_actor))))))

(event Custom|CleanupBuildCheck
  (Utilities|Time|ClearandInvalidateTimerbyHandle (Variables|Default|GetCheckBuildTimerHandle)))

(event Custom|ForceSpawnSlingshot
  (params (PawnToPingFor :type Pawn%20Object%20Reference))
  (bind _returnvalue (Inventory|SpawnPickupInWorld "/WildEstateWeapons/Gameplay/BeanDog/AGID_WildEstate_Launcher_BeanDog_UR.AGID_WildEstate_Launcher_BeanDog_UR" 1 (CallFunction|GetSlingshotSpawnLocation)))
  (MarkerComponent|MarkActoronServer (Actor|GetComponentbyClass (Fortnite|GetControllerfromActor PawnToPingFor) "/Script/FortniteGame.AthenaMarkerComponent") _returnvalue true true))

(event Custom|ForceSpawnSlurp
  (params (PawnToPingFor :type Pawn%20Object%20Reference))
  (CallFunction|ParsePlaylist)
  (Inventory|SpawnPickupInWorld "/WildEstateConsumables/Gameplay/NeonBurn/FoamFlood/WID_FoamFlood.WID_FoamFlood" (Variables|Default|GetNumSlurpToDrop) (CallFunction|GetSlingshotSpawnLocation) "0.000000,0.000000,2.000000"))

(event Custom|ParsePlaylist
  (bind _playlisttags (Variables|Config|GetPlaylistTags))
  (bind _array_element (Utilities|Array|ForEachLoopwithBreak (GameplayTags|BreakGameplayTagContainer (Class|FortGameStateAthena|GetAthenaPlaylistContextTags (Fortnite|Athena|GameState|GetGameStateAthena))))
    (:LoopBody
      (if (>= (Utilities|Array|FindItem _playlisttags _array_element) 0)
        (Variables|Default|SetNumSlurpToDrop (* (+ (Utilities|Array|FindItem _playlisttags _array_element) 1) (ScalableFloat|GetValueasInteger (Variables|Config|GetSlurpMultiplier) (Math|Conversions|ToFloat(Integer) (Utilities|Array|FindItem _playlisttags _array_element)))))))))

(fn BlueprintCanInteract
  (params (InteractingPawn :type Fort%20Pawn%20Object%20Reference) (InteractionBeingAttempted :type EInteractionBeingAttempted%20Enum :default "FirstInteraction") (InteractionType :type TInteractionType%20Enum :default "IT_NoInteraction"))
  (return false))

(fn GetSlingshotSpawnLocation
  (bind (_returnvalue_x _returnvalue_y _returnvalue_z) (Math|Vector|BreakVector (+ (Transformation|GetWorldLocation (Variables|Transformation|GetRootComponent)) (Variables|Container|GetLootSpawnLocation))))
  (return _returnvalue_x _returnvalue_y (+ _returnvalue_z 2e+01)))

(fn TurnCollisionOff
  (bind _explosionvfx (Variables|Default|GetExplosionVFX))
  (bind _audio (Variables|Default|GetAudio))
  (Networking|SwitchHasAuthority
    (:Authority
      (CallFunction|CleanupBuildCheck))
    (:Remote
      _))
  (Collision|SetCollisionEnabled (Variables|Default|GetLaunchCollision) "NoCollision")
  (Rendering|SetVisibility _explosionvfx false false)
  (Components|Activation|SetActive _explosionvfx false false)
  (if (Audio|Components|Audio|IsPlaying _audio)
    (Audio|Components|Audio|FadeOut _audio 0.2 0.0 "Linear" :!class /Script/Engine.AudioComponent)
    (else
      _))
  (Components|Activation|Deactivate _audio))

(fn TurnCollisionOn
  (bind _explosionvfx (Variables|Default|GetExplosionVFX))
  (bind _audio (Variables|Default|GetAudio))
  (Networking|SwitchHasAuthority
    (:Authority
      (CallFunction|CheckforBuilds))
    (:Remote
      _))
  (Rendering|SetVisibility (Variables|Debug|GetStaticMeshComponent) false false)
  (Rendering|SetVisibility (Variables|Default|GetBrokenStaticMesh) true false)
  (Collision|SetCollisionEnabled (Variables|Default|GetLaunchCollision) "QueryOnly")
  (Rendering|SetVisibility _explosionvfx true false)
  (Components|Activation|SetActive _explosionvfx true false)
  (Utilities|FlowControl|DoOnce false
    (:Completed
      (Audio|SpawnSoundAttached (Variables|Sound|GetVent%20Activate%20Sound) _audio)
      (Audio|SpawnSoundatLocation (Variables|Sound|GetLootDropSound) (Transformation|GetWorldLocation _audio))
      (Audio|Components|Audio|SetSound _audio (Variables|Sound|GetVentGasLeakLoop))
      (Audio|Components|Audio|Play _audio 0.0))))

(fn ConstructionScript)
