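; BGA_BasilOkra_Button_2 is a Fortnite gameplay actor for a button that shows a "Press Meenie" interaction prompt.
; Its interaction-string override returns that localized text regardless of the interacting pawn or interaction type.
; BeginPlay, overlap, Tick and the construction script are empty, and the file shows no logic for what a press does.
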(event EventBeginPlay)

(event Collision|EventActorBeginOverlap
  (params (OtherActor :type Actor%20Object%20Reference)))

(event EventTick
  (params (DeltaSeconds :type Float%20%28single-precision%29)))

(fn BlueprintGetInteractionString
  (params (InteractingPawn :type Fort%20Pawn%20Object%20Reference) (InteractionBeingAttempted :type EInteractionBeingAttempted%20Enum :default "FirstInteraction"))
  (return "NSLOCTEXT(\"[BBF446D19F95481C9B1EE0C1778454D2]\", \"69BF4FB8499BF7756FBB80A98906AF09\", \"Press Meenie\")"))

(fn ConstructionScript)
