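; Base interactable console button for BasilOkra, usable while bCanInteract is set and the pawn is not DBNO.
; Focus, interact and replication callbacks drive SetButtonVisualState, which sets mesh visibility and switches dim, bright or pulse looks through custom primitive data.
; Custom events swap green success or red fail materials and restore the cached default, with success hiding the mesh on a timer.
; Scalable floats control interact duration and prompt display.
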(event EventBeginPlay
  (|CallParentFunction)
  (Variables|Default|SetDefault_Mat (Rendering|Material|GetMaterial (Variables|Default|GetButtonMesh) 0))
  (Rendering|SetVisibility (Variables|Default|GetButtonMesh) false false))

(event InteractInterface|EventOnInteractFocused
  (params (FortPC :type Fort%20Player%20Controller%20Object%20Reference) (ContextInfo :type Fort%20Interact%20Context%20Info%20Object%20Reference))
  (CallFunction|SetButtonVisualState (Math|Integer|SelectInt 2 0 (Variables|Default|GetbCanInteract)))
  (bind _asfort_player_controller_athena (Utilities|Casting|CastToFortPlayerControllerAthena FortPC))
  (Utilities|FlowControl|DoOnce false
    (:Completed
      (if (Pawn|IsLocallyViewed (Pawn|GetControlledPawn _asfort_player_controller_athena))
        (Audio|SpawnSoundAttached (Variables|Audio|GetHighlightSound) (Variables|Default|GetButtonMesh) "None" "0.000000,0.000000,30.000000")))))

(event InteractInterface|EventOnInteractUnfocused
  (params (FortPC :type Fort%20Player%20Controller%20Object%20Reference) (NewContextInfo :type Fort%20Interact%20Context%20Info%20Object%20Reference))
  (CallFunction|SetButtonVisualState (Math|Integer|SelectInt 1 0 (Variables|Default|GetbCanInteract)))
  (Utilities|FlowControl|DoOnce false
    (:Completed
      (if (Pawn|IsLocallyViewed (Pawn|GetControlledPawn (stale InteractInterface|EventOnInteractFocused _asfort_player_controller_athena)))
        (Audio|SpawnSoundAttached (Variables|Audio|GetHighlightSound) (Variables|Default|GetButtonMesh) "None" "0.000000,0.000000,30.000000")))))

(event Interaction|EventBlueprintOnBeginInteract
  (params (InteractingPawn :type Fort%20Pawn%20Object%20Reference) (InteractionBeingAttempted :type EInteractionBeingAttempted%20Enum :default "FirstInteraction") (InteractionType :type TInteractionType%20Enum :default "IT_NoInteraction"))
  (|CallParentFunction InteractingPawn InteractionBeingAttempted InteractionType)
  (CallFunction|SetButtonVisualState 3))

(event Interaction|EventBlueprintOnInterruptInteract
  (|CallParentFunction)
  (CallFunction|SetButtonVisualState 1))

(event Custom|Success_State
  ;[1 1 1] to avoid fighting with the remove interact calls
  (Utilities|FlowControl|DelayUntilNextFrame)
  (Rendering|Material|SetMaterial (Variables|Default|GetButtonMesh) 0 "/WildEstate_BasilOkra/Assets/Console/MI_BasilOkra_ConsoleInteraction_Green.MI_BasilOkra_ConsoleInteraction_Green")
  (CallFunction|SetButtonVisualState 2)
  (Rendering|SetVisibility (Variables|Default|GetButtonMesh) true false)
  (Utilities|Time|SetTimerbyEvent (EventDispatchers|CreateEvent :!event Custom|Cleanup) 5.0)
  (Development|PrintString "Play Success Material" true true "(R=0.000000,G=0.660000,B=1.000000,A=1.000000)" 15.0))

(event Custom|Fail_State
  (Rendering|Material|SetMaterial (Variables|Default|GetButtonMesh) 0 "/WildEstate_BasilOkra/Assets/Console/MI_BasilOkra_ConsoleInteraction_Red.MI_BasilOkra_ConsoleInteraction_Red")
  (CallFunction|SetButtonVisualState 3)
  (Rendering|SetVisibility (Variables|Default|GetButtonMesh) true false)
  (Development|PrintString "Play Failure Material " true true "(R=0.000000,G=0.660000,B=1.000000,A=1.000000)" 15.0))

(event Custom|Reset_State
  (Rendering|Material|SetMaterial (Variables|Default|GetButtonMesh) 0 (Variables|Default|GetDefault_Mat))
  (CallFunction|SetButtonVisualState 2)
  (Development|PrintString "Reset Called Properly" true true "(R=0.000000,G=0.660000,B=1.000000,A=1.000000)" 1e+01))

(event Custom|Cleanup
  (Rendering|SetVisibility (Variables|Default|GetButtonMesh) false false))

(fn BlueprintCanInteract
  (params (InteractingPawn :type Fort%20Pawn%20Object%20Reference) (InteractionBeingAttempted :type EInteractionBeingAttempted%20Enum :default "FirstInteraction") (InteractionType :type TInteractionType%20Enum :default "IT_NoInteraction"))
  (return (and (Variables|Default|GetbCanInteract) (not (Health|IsDBNO InteractingPawn)))))

(fn BlueprintGetAllHighlightableComponents
  (return (Utilities|Array|MakeArray (Variables|Default|GetButtonMesh)) false))

(fn BlueprintGetInteractionString
  (params (InteractingPawn :type Fort%20Pawn%20Object%20Reference) (InteractionBeingAttempted :type EInteractionBeingAttempted%20Enum :default "FirstInteraction"))
  (return "NSLOCTEXT(\"[B6152A3527505251FECF66C7EAFD074C]\", \"7504BC154BB5AFFB647F00A2A581CE92\", \"Press Eenie\")"))

(fn BlueprintGetInteractionTime
  (params (InteractingPawn :type Fort%20Pawn%20Object%20Reference) (InteractionBeingAttempted :type EInteractionBeingAttempted%20Enum :default "FirstInteraction"))
  (return (ScalableFloat|GetValueatLevel (Variables|Default|GetInteractDuration)) true))

(fn OnRep_bCanInteract)

(fn OnRep_bShouldTurnOffMesh
  (CallFunction|SetButtonVisualState (Math|Integer|SelectInt 2 0 (not (Variables|Default|GetbShouldTurnOffMesh)))))

(fn SetButtonVisualState
  (params (State :type Integer))
  (Rendering|SetVisibility (Variables|Default|GetButtonMesh) (and (not (== State 0)) (ScalableFloat|GetValueasBool (Variables|Default|GetShouldShowInteractString))) false)
  (switch int (Variables|GetState)
    (:Default
      _)
    (:0
      _)
    (:1
      _)
    (:2
      ;[1 1 1] StaticBright
      (Rendering|Material|SetCustomPrimitiveDataFloat (Variables|Default|GetButtonMesh) 0 1.0))
    (:3
      ;[1 1 1] Pulse
      (Rendering|Material|SetCustomPrimitiveDataFloat (Variables|Default|GetButtonMesh) 0 2.0)))
  ;[1 1 1] Static Dim
  (Rendering|Material|SetCustomPrimitiveDataFloat (Variables|Default|GetButtonMesh) 0 0.0))

(fn ShouldDisplayFirstInteractPrompt
  (params (InteractingPawn :type Fort%20Pawn%20Object%20Reference) (InteractionType :type TInteractionType%20Enum :default "IT_NoInteraction"))
  (return (ScalableFloat|GetValueasBool (Variables|Default|GetShouldShowInteractString))))

(fn ConstructionScript)
