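; Console actor for the BasilOkra meltdown puzzle, where players press three buttons in the order shown on spawned instruction manuals before a countdown expires.
; The authority shuffles the sequence, runs timers and penalizes wrong attempts; success spawns loot and fires a squad quest, while timeout starts a timed meltdown.
; A replicated state enum's OnRep drives interaction, markers, UI, camera shake, post process and room-aware music.
; Timer digits and button feedback display through a Material Parameter Collection.
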(event OnListenerEnterEvent(AGV_LootRoom)
  ;[0 0.2967 0.3432 0.5] Sound
  (Variables|Sound|SetPlayerInsideLootRoom? true))

(event OnToggledOn(AudioGameplayVolume)
  ;[0 0.2967 0.3432 0.5] Sound
  ;;[1 1 1] Handle music when players enter/exit the console room
  (Variables|Sound|SetbIsInsideConsoleRoom true)
  (if (and (Variables|Sound|GetbIsInsideConsoleRoom) (Variables|Sound|GetbCountdownTimerIsActive?))
    (bind _musicloopref (Variables|Default|GetMusicLoopRef)
      (:then
        (Audio|Components|Audio|AdjustVolume _musicloopref (Variables|Sound|GetMusicFadeTime) 1.0 "Linear" :!class /Script/Engine.AudioComponent))
      (:else
        (bind _returnvalue (Audio|SpawnSoundAttached (Variables|Sound|GetMusic%20Meltdown) (Variables|Default|GetAudio_Console)))
        (bind _output_get (Variables|Default|SetMusicLoopRef _returnvalue))
        (Audio|Parameter|SetFloatParameter _output_get "StartTime" (Math|Random|RandomFloatinRange 3.0 5e+01) :!class /Script/AudioExtensions.AudioParameterControllerInterface)
        (Audio|AddEventbyComponent (GameInstanceSubsystems|GetMusicEventSubsystem) _output_get "(TagName=\"MusicEventSubsystem.Event.POI.Special\")" "(GameplayTags=)" 0 "")))
    (else
      (CallFunction|StopLoopingMusic (Variables|Sound|GetMusicFadeTime)))))

(event OnListenerExitEvent(AGV_LootRoom)
  ;[0 0.2967 0.3432 0.5] Sound
  (Variables|Sound|SetPlayerInsideLootRoom? false))

(event OnToggledOff(AudioGameplayVolume)
  ;[0 0.2967 0.3432 0.5] Sound
  ;;[1 1 1] Handle music when players enter/exit the console room
  (Variables|Sound|SetbIsInsideConsoleRoom false)
  (if (and (Variables|Sound|GetbIsInsideConsoleRoom) (Variables|Sound|GetbCountdownTimerIsActive?))
    (bind _musicloopref (Variables|Default|GetMusicLoopRef)
      (:then
        (Audio|Components|Audio|AdjustVolume _musicloopref (Variables|Sound|GetMusicFadeTime) 1.0 "Linear" :!class /Script/Engine.AudioComponent))
      (:else
        (bind _returnvalue (Audio|SpawnSoundAttached (Variables|Sound|GetMusic%20Meltdown) (Variables|Default|GetAudio_Console)))
        (bind _output_get (Variables|Default|SetMusicLoopRef _returnvalue))
        (Audio|Parameter|SetFloatParameter _output_get "StartTime" (Math|Random|RandomFloatinRange 3.0 5e+01) :!class /Script/AudioExtensions.AudioParameterControllerInterface)
        (Audio|AddEventbyComponent (GameInstanceSubsystems|GetMusicEventSubsystem) _output_get "(TagName=\"MusicEventSubsystem.Event.POI.Special\")" "(GameplayTags=)" 0 "")))
    (else
      (CallFunction|StopLoopingMusic (Variables|Sound|GetMusicFadeTime)))))

(event Custom|StopLoopingMusic
  (params (FadeOutTime :type Float%20%28double-precision%29))
  ;[0 0.2967 0.3432 0.5] Sound
  ;;[1 1 1] Stop Looping Music
  (bind _musicloopref (Variables|Default|GetMusicLoopRef)
    (:then
      (Audio|Components|Audio|AdjustVolume _musicloopref FadeOutTime 0.01 "Linear" :!class /Script/Engine.AudioComponent)
      (Audio|RemoveEvent (GameInstanceSubsystems|GetMusicEventSubsystem) _musicloopref))
    (:else)))

(event Custom|PlayMeltdownBuildup
  ;[0 0.2967 0.3432 0.5] Sound
  ;;[1 1 1] Buildup Audio (Fakeout)
  (Networking|CanExecuteCosmeticEvents
    (:True
      (Audio|SpawnSoundatLocation (Variables|Sound|GetMeltdownBuildupFakeOut) (Transformation|GetWorldLocation (Variables|Default|GetAudio_Console))))))

(event Custom|PlayShutdownAudio
  (bind _audio_console (Variables|Default|GetAudio_Console))
  (bind _returnvalue (Transformation|GetWorldLocation _audio_console))
  ;[0 0.2967 0.3432 0.5] Sound
  ;;[1 1 1] Shutdown Audio Sequence
  (Networking|CanExecuteCosmeticEvents
    (:True
      (if (Variables|Sound|GetbPuzzleSucceeded?)
        (if (and (Pawn|IsLocallyViewed (Variables|Default|GetInteracting%20Pawn)) (Variables|Sound|GetPlayerInsideLootRoom?))
          (Audio|Modulation|ActivateControlBusMix (Variables|Sound|GetCBM_PuzzleSucceed))
          (else
            _))
        (Audio|SpawnSoundatLocation (Variables|Sound|GetAudio%20Shutdown%20Sequence_Succeed) _returnvalue)
        (else
          (Utilities|FlowControl|Delay 9.0)
          (Audio|SpawnSoundatLocation (Variables|Sound|GetAudioShutdownSequence_Fail) _returnvalue))))))

(event Custom|PlayMusicSuccessStinger
  ;[0 0.2967 0.3432 0.5] Sound
  ;;[1 1 1] Success Music Stinger
  (Audio|SpawnSoundatLocation (Variables|Sound|GetPuzzleSolvedFlourishSound) (Transformation|GetWorldLocation (Variables|Default|GetAudio_Console)))
  (if (and (Pawn|IsLocallyViewed (Variables|Default|GetInteracting%20Pawn)) (Variables|Sound|GetPlayerInsideLootRoom?))
    (bind _returnvalue (Audio|SpawnSound2D (Variables|Sound|GetMusicStinger%20Succeed)))
    (Variables|Default|SetMusicStingRef _returnvalue)
    (else
      _))
  (bind _musicstingref (Variables|Default|GetMusicStingRef)
    (:then
      (Audio|AddEventbyComponent (GameInstanceSubsystems|GetMusicEventSubsystem) _musicstingref "(TagName=\"MusicEventSubsystem.Event.POI.Special\")" "(GameplayTags=)" 0 "")
      (bind _returnvalue_2 (Utilities|Time|SetTimerbyEvent (delegate Custom|UnregisterStingerEvent) 9.0))
      (Variables|Sound|SetStingerTimerHandle _returnvalue_2))
    (:else)))

(event Custom|UnregisterStingerEvent
  ;[0 0.2967 0.3432 0.5] Sound
  (bind _musicstingref (Variables|Default|GetMusicStingRef)
    (:then
      (Audio|RemoveEvent (GameInstanceSubsystems|GetMusicEventSubsystem) _musicstingref))
    (:else
      _))
  (Utilities|Time|ClearandInvalidateTimerbyHandle (Variables|Sound|GetStingerTimerHandle)))

(event Custom|PlayMusicFailStinger
  ;[0 0.2967 0.3432 0.5] Sound
  ;;[1 1 1] Time Exceeded Fail Stinger
  (if (and (Pawn|IsLocallyViewed (Variables|Default|GetInteracting%20Pawn)) (Variables|Sound|GetPlayerInsideLootRoom?))
    (bind _returnvalue (Audio|SpawnSound2D (Variables|Sound|GetMusicStinger%20Fail)))
    (Variables|Default|SetMusicStingRef _returnvalue)
    ;;[1 1 1] Success Music Stinger
    (bind _musicstingref (Variables|Default|GetMusicStingRef)
      (:then
        (Audio|AddEventbyComponent (GameInstanceSubsystems|GetMusicEventSubsystem) _musicstingref "(TagName=\"MusicEventSubsystem.Event.POI.Special\")" "(GameplayTags=)" 0 "")
        (bind _returnvalue_2 (Utilities|Time|SetTimerbyEvent (delegate Custom|UnregisterStingerEvent) 9.0))
        (Variables|Sound|SetStingerTimerHandle _returnvalue_2))
      (:else))))

(event Custom|ForceStartEvent
  (CallFunction|BasilOkraSetCountdownActive))

(event EventBeginPlay
  (|CallParentFunction)
  (Class|FortAsyncActionCurrentPlaylistReady|PlaylistDataReadyAsync
    (:then
      (Networking|SwitchHasAuthority
        (:Authority
          (Default|InitializeButtonActors)
          (CallFunction|SpawnLootVent)
          (Class|BMMObjObjectiveActorComponent|RegisterObjectiveforAutoMarking (Variables|Default|GetB_MMObj_ObjectiveActorComponent))
          (if (ScalableFloat|GetValueasBool (Variables|Config|Console|GetShouldForceStart))
            (bind _returnvalue (Utilities|Time|SetTimerbyEvent (delegate Custom|ForceStartEvent) (ScalableFloat|GetValueatLevel (Variables|Config|Console|GetForceStartTime))))
            (Variables|Default|SetForceStartTimerHandle _returnvalue)))))
    (:OnReady
      (CallFunction|SetInteractState false true)
      (Components|Activation|SetActive (Variables|Default|GetSequenceFailVFX) false true)
      (for _array_element (Variables|Default|GetButtonActorArray)
        (Utilities|Array|AddUnique (Variables|Transient|Console|GetCorrectButtonTagArray) (Variables|/WildEstate_BasilOkra/Gameplay/Actors/BGA_BasilOkra_Button_Base.BGA_BasilOkra_Button_Base_C|GetButtonTag _array_element))
        (Interaction|BindEventtoOnInteract _array_element (EventDispatchers|CreateEvent :!event Custom|DoOnInteract)))
      (Utilities|FlowControl|Sequence
        (:then_0
          (CallFunction|BasilOkraSpawnInstructions)
          (if (ScalableFloat|GetValueasBool (Variables|Config|Instructions|GetShouldRandomizeLocationAtStart))
            ;[1 1 1] Making this manual for now just for testing purposes
            (CallFunction|BasilOkraShuffleInstructionLocation)))
        (:then_1
          (Networking|SwitchHasAuthority
            (:Authority
              (if (ScalableFloat|GetValueasBool (Variables|Config|Instructions|GetShouldRandomizeSequenceAtStart))
                (CallFunction|BasilOkraShuffleSequence)
                (else
                  (Variables|Transient|Console|SetCorrectButtonTagArray (Variables|Transient|Console|GetCorrectButtonTagArray)))))))))))

(event Custom|TurnPostProcessOn
  ;[1 1 1] Post Prcoess On off
  (bind (_direction _alpha) (Timeline|TogglePost Play)
    (:Update
      (Variables|Transient|SetPostProcessAlpha _alpha))
    (:Finished
      (Variables|Transient|SetPostProcessAlpha (Utilities|Select :Index _direction :Forward 1.0 :Backward 0.0)))))

(event Custom|TurnPostProcessOff
  ;[1 1 1] Post Prcoess On off
  (Timeline|TogglePost Reverse))

(event Custom|UpdateSequenceVisualAll
  (params (ButtonPressed :type Gameplay%20Tag%20Structure) (SequenceIndex :type Integer))
  ;[1 0 0 0.5] TODO - Maybe Update this section, can probably change.
  ; Can likely change "Current Button Tag Array" to be RepNotify and work this in there
  (CallFunction|SetSequenceVisual ButtonPressed SequenceIndex))

(event Custom|DoOnInteract
  (params (SelfActor :type Building%20Actor%20Object%20Reference) (InteractingPawn :type Fort%20Pawn%20Object%20Reference))
  (Utilities|FlowControl|Sequence
    (:then_0
      (if (Utilities|Enum|Equal(Enum) (Variables|Transient|Console|GetBasilOkra_State_Enum) "NewEnumerator1")
        (bind _asbga_basil_okra_button_base (Utilities|Casting|CastToBGA_BasilOkra_Button_Base SelfActor))
        (CallFunction|BroadcastInteracttoScreens)
        (bind _returnvalue (Utilities|Array|Add (Variables|Transient|Console|GetCurrentButtonTagArray) (Variables|/WildEstate_BasilOkra/Gameplay/Actors/BGA_BasilOkra_Button_Base.BGA_BasilOkra_Button_Base_C|GetButtonTag _asbga_basil_okra_button_base)))
        (if (== (+ _returnvalue 1) (Utilities|Array|Length (Variables|Transient|Console|GetCorrectButtonTagArray)))
          (Variables|Transient|SetInteractor InteractingPawn)
          (CallFunction|CheckSequence)
          (else
            (Variables|/WildEstate_BasilOkra/Gameplay/Actors/BGA_BasilOkra_Button_Base.BGA_BasilOkra_Button_Base_C|SetbCanInteract false _asbga_basil_okra_button_base)))))
    (:then_1
      (Variables|Default|SetInteracting%20Pawn InteractingPawn))
    (:then_2
      (GameplayCue|ExecuteGameplayCueOnActor(Burst) SelfActor (Variables|Default|GetInteractButtonGCTag) (Ability|GameplayCue|MakeGameplayCueParameters 0.0)))))

(event Custom|BroadcastInteractToScreens
  (for _array_element (Variables|Config|Console|GetConsoleScreenInstanceArray)
    (Class|BGABasilOkraOverheadConsoleA|DoonInteractSuccess _array_element)))

(event Custom|CheckSequence
  (bind (_array_element _array_index) (Utilities|Array|ForEachLoopwithBreak (Variables|Transient|Console|GetCorrectButtonTagArray))
    (:LoopBody
      (if (not (== _array_element (Utilities|Array|Get(aref) (Variables|Transient|Console|GetCurrentButtonTagArray) _array_index)))
        (Variables|Transient|Console|SetbIsSequenceCorrect false)))
    (:Completed
      _))
  (if (Variables|Transient|Console|GetbIsSequenceCorrect)
    (Variables|Transient|Console|SetBasilOkra_State_Enum "NewEnumerator2")
    (Development|PrintString "SEQUENCE SUCCESS !!!")
    (else
      (CallFunction|DoonSequenceFailure)
      (Variables|Transient|SetInteractor 0)
      (Development|PrintString "SEQUENCE FAILED !!!"))))

(event Custom|DoOnSequenceSuccess
  (bind _spawnedlootcontainer (Variables|Transient|LootVent|GetSpawnedLootContainer))
  (Utilities|FlowControl|Sequence
    (:then_0
      (Networking|CanExecuteCosmeticEvents
        (:True
          (Components|Activation|SetActive (Variables|Default|GetSequenceSuccessVFX) true true))))
    (:then_1
      (Networking|SwitchHasAuthority
        (:Authority
          (Class|BGABasilOkraLootContainerInvisible|SpawnLootbyTierGroup _spawnedlootcontainer)
          (CallFunction|FireQuest)
          (for _array_element (Variables|Default|GetButtonActorArray)
            (Class|BGABasilOkraButtonBase|SuccessState _array_element)))
        (:Remote
          (Utilities|FlowControl|DoOnce false
            (:Completed
              (Audio|SpawnSoundatLocation (Variables|Sound|GetLootDropSound) (Transformation|GetActorLocation _spawnedlootcontainer)))))))
    (:then_2
      (Rendering|Components|TextRender|SetText (Variables|Default|GetCountdown_TextRender) (Utilities|Text|MakeLiteralText "NSLOCTEXT(\"[6DF35C299CF81AABA04BBF2C5CF22E44]\", \"BC8B17044EA4704BCC3045B17AFCD157\", \"LUCKY GUESS. . .\")")))
    (:then_3
      (GameplayCue|ExecuteGameplayCueOnActor(Burst) self (Variables|Default|GetConfettiGCTag) (Ability|GameplayCue|MakeGameplayCueParameters 0.0)))))

(event Custom|ActivateMeltdownVisuals
  (Console|SetAllWarningStateVisuals true)
  (Rendering|Material|SetScalarParameterValue "/WildEstateTech/TechArt/GlowDonut/Timer/MPC_GlowDonutState.MPC_GlowDonutState" "CountdownIsVisible" 1.0 :!class /Script/Engine.KismetMaterialLibrary)
  (CallFunction|ActivateRegisteredVisualActors)
  (Networking|CanExecuteCosmeticEvents
    (:True
      (Class|BPBasilOkraVisualCollision|ActivateVisualsThroughInterface (Variables|Config|GetVisualCollisionInstance)))))

(event Custom|DoOnSequenceFailure
  (if (> (ScalableFloat|GetValueatLevel (Variables|Config|Console|GetCountdownPenalty)) 0)
    (CallFunction|PenalizeCountdown))
  (Utilities|Array|Clear (Variables|Transient|Console|GetCurrentButtonTagArray))
  (Variables|Transient|Console|SetbIsSequenceCorrect true)
  (bind _returnvalue (Utilities|Time|SetTimerbyEvent (EventDispatchers|CreateEvent :!event Custom|DoOnFailureDelayCompleted) (ScalableFloat|GetValueatLevel (Variables|Config|Console|GetSequenceFailureDelayDuration))))
  (Variables|Default|SetSequenceFailureDelayTimer _returnvalue)
  (for _array_element (Variables|Default|GetButtonActorArray)
    (Class|BGABasilOkraButtonBase|FailState _array_element))
  (Networking|CanExecuteCosmeticEvents
    (:True
      (CallFunction|DoonSequenceFailureCosmeticOnly))))

(event Custom|DeactivateMeltdownVisuals
  (Utilities|FlowControl|Sequence
    (:then_0
      (CallFunction|DeactivateRegisteredVisualActors)
      (Networking|CanExecuteCosmeticEvents
        (:True
          (Class|BPBasilOkraVisualCollision|DeactivateVisualsThroughInterface (Variables|Config|GetVisualCollisionInstance)))))
    (:then_1
      (Console|SetAllWarningStateVisuals false)
      (Rendering|Material|SetScalarParameterValue "/WildEstateTech/TechArt/GlowDonut/Timer/MPC_GlowDonutState.MPC_GlowDonutState" "CountdownIsVisible" :!class /Script/Engine.KismetMaterialLibrary)
      (Rendering|Material|SetScalarParameterValue "/WildEstateTech/TechArt/GlowDonut/Timer/MPC_GlowDonutState.MPC_GlowDonutState" "PlayMonitorAnims" 1.0 :!class /Script/Engine.KismetMaterialLibrary)
      (bind (_direction _alpha) (Timeline|EndMonitorAnim PlayFromStart)
        (:Update
          (Rendering|Material|SetScalarParameterValue "/WildEstateTech/TechArt/GlowDonut/Timer/MPC_GlowDonutState.MPC_GlowDonutState" "MonitorAnimPlayAlpha" _alpha :!class /Script/Engine.KismetMaterialLibrary))
        (:Finished
          (Rendering|Material|SetScalarParameterValue "/WildEstateTech/TechArt/GlowDonut/Timer/MPC_GlowDonutState.MPC_GlowDonutState" "PlayMonitorAnims" :!class /Script/Engine.KismetMaterialLibrary))))))

(event Custom|DoOnSequenceFailure_CosmeticOnly
  (Audio|PlaySoundatLocation (Variables|Config|Console|Cosmetic|GetSequenceFailedSound) (Transformation|GetActorLocation))
  (Components|Activation|SetActive (Variables|Default|GetSequenceFailVFX) true true)
  (Rendering|Material|SetScalarParameterValue "/WildEstateTech/TechArt/GlowDonut/Timer/MPC_GlowDonutState.MPC_GlowDonutState" "ShowFailDisplay" 1.0 :!class /Script/Engine.KismetMaterialLibrary)
  (Audio|PlaySoundatLocation (Variables|Sound|GetFailed_Sweetener) (Transformation|GetActorLocation)))

(event Custom|ActivateMeltdownUI
  (Class|BGABasilOkraMMOInteractionParent|BasilOkraActivateObjective (Variables|Config|Console|GetUIActorInstance)))

(event Custom|DoOnFailureDelayCompleted
  (for _array_element (Variables|Default|GetButtonActorArray)
    (Variables|/WildEstate_BasilOkra/Gameplay/Actors/BGA_BasilOkra_Button_Base.BGA_BasilOkra_Button_Base_C|SetbCanInteract true _array_element)
    (Class|BGABasilOkraButtonBase|ResetState _array_element))
  (Utilities|Time|ClearandInvalidateTimerbyHandle (Variables|Default|GetSequenceFailureDelayTimer))
  (Networking|CanExecuteCosmeticEvents
    (:True
      (Rendering|Material|SetScalarParameterValue "/WildEstateTech/TechArt/GlowDonut/Timer/MPC_GlowDonutState.MPC_GlowDonutState" "ShowFailDisplay" :!class /Script/Engine.KismetMaterialLibrary)
      (CallFunction|ResetSequenceVisualstoWarning))))

(event Custom|DeactivateMeltdownUI
  (Class|BGABasilOkraMMOInteractionParent|BasilOkraForceObjectiveTimerEnd (Variables|Config|Console|GetUIActorInstance)))

(event Custom|SetEDLState
  (params (DLState :type EDataLayerRuntimeState%20Enum :default "Unloaded"))
  (bind _returnvalue (WorldPartition|GetDataLayerManager))
  (bind _returnvalue_2 (DataLayers|SetDataLayerInstanceRuntimeState _returnvalue (DataLayers|GetDataLayerInstancefromAsset _returnvalue "/WildEstate_BasilOkra/Levels/EDL_WildEstate_BasilOkra.EDL_WildEstate_BasilOkra") DLState false))
  (Development|PrintString (Utilities|String|ToString(Boolean) _returnvalue_2)))

(event Custom|BroadcastGlobalEventActivated
  (Networking|CanExecuteCosmeticEvents
    (:True
      (GameplayEventRouter|BroadcastEventBackwardCompatible (GameplayEventRouter|GetGlobalEventRouter) "(TagName=\"EventMessage.Athena.BasilOkra.ObjectiveActivated\")" (Utilities|Struct|MakeTrainMMOObjectiveActivatedBasilOkra self "Neutral" (Variables|Transient|Timer|GetCountdownRemaining))))))

(event Custom|BroadcastGlobalEventSuccess
  (Networking|CanExecuteCosmeticEvents
    (:True
      (GameplayEventRouter|BroadcastEventBackwardCompatible (GameplayEventRouter|GetGlobalEventRouter) "(TagName=\"EventMessage.Athena.BasilOkra.ObjectiveSuccess\")" (Utilities|Struct|MakeTrainMMOObjectiveActivatedBasilOkra self "Neutral" (Variables|Transient|Timer|GetCountdownRemaining))))))

(event Custom|CleanupGlobalMarker
  (Class|BMMObjObjectiveActorComponent|UnregisterObjectiveforAutoMarking (Variables|Default|GetB_MMObj_ObjectiveActorComponent))
  (bind _globalmapiconactor (Variables|Config|GetGlobalMapIconActor)
    (:then
      (Actor|DestroyActor _globalmapiconactor)
      (MarkerComponent|DeactivateMarkersonActor self))
    (:else)))

(event Custom|FireQuest
  (bind _asfort_player_state_athena (Utilities|Casting|CastToFortPlayerStateAthena (Variables|/Script/Engine.Pawn|GetPlayerState (Variables|Default|GetInteracting%20Pawn))))
  (bind _returnvalue (Class|FortGameStateAthena|GetSquadMemberPlayerStates (Fortnite|Athena|GameState|GetGameStateAthena) _asfort_player_state_athena true))
  (for _array_element _returnvalue
    (CustomVerb|BroadcastCustomVerbMessage (CustomVerb|MakeCustomVerbMessage (Fortnite|GetFortPlayerControllerfromActor _array_element) "(TagName=\"CustomVerb.WildEstate.PowerPlant.MMOComplete\")"))))

(event Custom|SpawnGlobalMarker
  (bind _returnvalue (Game|SpawnActorfromClass "/WildEstate_BasilOkra/Gameplay/Actors/B_BasilOkra_GlobalMapIconActor.B_BasilOkra_GlobalMapIconActor_C" (Math|Conversions|ToTransform(Vector) (Transformation|GetActorLocation)) "AlwaysSpawn" "OverrideRootScale" 0 0))
  (Variables|Config|SetGlobalMapIconActor _returnvalue)
  (if (not (ScalableFloat|GetValueasBool (Variables|Default|GetShouldPingOnProximity)))
    (CallFunction|MarkGlobalScreenPing)))

(event Custom|StartCameraShake
  ;[1 1 1] Camera Shake Logic
  (bind _boxcollision (Class|BPBasilOkraVisualCollision|GetPawnCollision (Variables|Config|GetVisualCollisionInstance)))
  (Collision|BindEventtoOnComponentBeginOverlap _boxcollision (delegate Custom|ApplyCameraShake))
  (Collision|BindEventtoOnComponentEndOverlap _boxcollision (delegate Custom|RemoveCameraShake))
  (CallFunction|InitialApplyCameraShake))

(event Custom|ActivateRegisteredVisualActors
  (for _array_element (Variables|Transient|GetRegisteredVisualActorsArray)
    (Default|CallActivateVisuals (Actor|GetComponentbyClass _array_element "/WildEstate_BasilOkra/Gameplay/Actors/Visual/BPAC_BasilOkra_VisualActivationComponent.BPAC_BasilOkra_VisualActivationComponent_C"))))

(event Custom|ApplyCameraShake
  (params (OverlappedComponent :type Primitive%20Component%20Object%20Reference) (OtherActor :type Actor%20Object%20Reference) (OtherComp :type Primitive%20Component%20Object%20Reference) (OtherBodyIndex :type Integer) (bFromSweep :type Boolean) (SweepResult :type Hit%20Result%20Structure%20%28by%20ref%29))
  ;[1 1 1] Camera Shake Logic
  (GameplayCue|AddGameplayCue OtherActor (Variables|Default|GetCameraShakeGCTag)))

(event Custom|RemoveCameraShake
  (params (OverlappedComponent :type Primitive%20Component%20Object%20Reference) (OtherActor :type Actor%20Object%20Reference) (OtherComp :type Primitive%20Component%20Object%20Reference) (OtherBodyIndex :type Integer))
  ;[1 1 1] Camera Shake Logic
  (GameplayCue|RemoveGameplayCue OtherActor (Variables|Default|GetCameraShakeGCTag)))

(event Custom|DeactivateRegisteredVisualActors
  (for _array_element (Variables|Transient|GetRegisteredVisualActorsArray)
    (Default|CallDeactivateVisuals (Actor|GetComponentbyClass _array_element "/WildEstate_BasilOkra/Gameplay/Actors/Visual/BPAC_BasilOkra_VisualActivationComponent.BPAC_BasilOkra_VisualActivationComponent_C"))))

(event Custom|TerminateCameraShake
  ;[1 1 1] Camera Shake Logic
  (bind _boxcollision (Class|BPBasilOkraVisualCollision|GetPawnCollision (Variables|Config|GetVisualCollisionInstance)))
  (Collision|UnbindEventfromOnComponentBeginOverlap _boxcollision (delegate Custom|ApplyCameraShake))
  (Collision|UnbindEventfromOnComponentEndOverlap _boxcollision (delegate Custom|RemoveCameraShake))
  (CallFunction|CleanupCameraShake))

(event Custom|InitialApplyCameraShake
  ;[1 1 1] Camera Shake Logic
  (bind _boxcollision (Class|BPBasilOkraVisualCollision|GetPawnCollision (Variables|Config|GetVisualCollisionInstance)))
  (for _array_element (Collision|GetOverlappingActors _boxcollision "/Script/FortniteGame.FortPlayerPawn" :!class /Script/Engine.PrimitiveComponent)
    (GameplayCue|AddGameplayCue _array_element (Variables|Default|GetCameraShakeGCTag))))

(event Custom|CleanupCameraShake
  ;[1 1 1] Camera Shake Logic
  (bind _boxcollision (Class|BPBasilOkraVisualCollision|GetPawnCollision (Variables|Config|GetVisualCollisionInstance)))
  (for _array_element (Collision|GetOverlappingActors _boxcollision "/Script/FortniteGame.FortPlayerPawn" :!class /Script/Engine.PrimitiveComponent)
    (GameplayCue|RemoveGameplayCue _array_element (Variables|Default|GetCameraShakeGCTag))))

(event Custom|BasilOkra_ShuffleSequence
  (bind _correctbuttontagarray (Variables|Transient|Console|GetCorrectButtonTagArray))
  (Utilities|Array|Shuffle _correctbuttontagarray)
  (Variables|Transient|Console|SetCorrectButtonTagArray _correctbuttontagarray))

(event Custom|BasilOkra_ShuffleInstructionLocation
  (Networking|SwitchHasAuthority
    (:Authority
      (Variables|Transient|Instructions|SetCachedInstructionLocations (Instructions|GetRandomInstructionManualLocation (Utilities|Array|Length (Variables|Transient|Instructions|GetInstructionsArray)))))))

(event Custom|BasilOkra_SpawnInstructions
  (Networking|SwitchHasAuthority
    (:Authority
      (for _index (range (Math|Float|Truncate (ScalableFloat|GetValueatLevel (Variables|Config|Instructions|GetNumPostersToSpawn))))
        (bind _spawnedinstructions (Class|ABasilOkraInstructionsSpawner|SpawnInstructionManual (Utilities|Array|Get(acopy) (Variables|Config|Instructions|GetInstructionSpawnArray) _index)))
        (Utilities|Array|Add (Variables|Transient|Instructions|GetInstructionsArray) _spawnedinstructions)))))

(event Custom|SetInteractState
  (params (CanInteract :type Boolean) (ShouldTurnOffMesh :type Boolean))
  (for _array_element (Variables|Default|GetButtonActorArray)
    (Variables|/WildEstate_BasilOkra/Gameplay/Actors/BGA_BasilOkra_Button_Base.BGA_BasilOkra_Button_Base_C|SetbCanInteract CanInteract _array_element)
    (Variables|/WildEstate_BasilOkra/Gameplay/Actors/BGA_BasilOkra_Button_Base.BGA_BasilOkra_Button_Base_C|SetbShouldTurnOffMesh ShouldTurnOffMesh _array_element)))

(event Custom|SpawnLootVent
  (bind _returnvalue (Game|SpawnActorfromClass (Variables|Config|LootVent|GetLootVentClass) (Transformation|GetWorldTransform (Variables|Default|GetLootSpawnLocation)) "AlwaysSpawn" "MultiplyWithRoot" self 0))
  (bind _asbga_basil_okra_loot_container_invisible (Utilities|Casting|CastToBGA_BasilOkra_LootContainer_Invisible _returnvalue))
  (Variables|Transient|LootVent|SetSpawnedLootContainer _asbga_basil_okra_loot_container_invisible))

(event Custom|CleanupTimers
  (Utilities|Time|ClearandInvalidateTimerbyHandle (Variables|Transient|Console|GetCountdownTimerHandle))
  (Utilities|Time|ClearandInvalidateTimerbyHandle (Variables|Transient|Console|GetCountdownUpdateTimerHandle))
  (Utilities|Time|ClearandInvalidateTimerbyHandle (Variables|Default|GetCountdownUpdateSecondsTimerHandle)))

(event Custom|SetUpCountdownTextRender
  (Networking|SwitchHasAuthority
    (:Authority
      (bind _returnvalue (Utilities|Time|SetTimerbyEvent (EventDispatchers|CreateEvent :!event Custom|TimerRender_Update) (ScalableFloat|GetValueatLevel (Variables|Config|Console|GetCountdownUpdateRate)) true))
      (Variables|Transient|Console|SetCountdownUpdateTimerHandle _returnvalue)
      (bind _returnvalue_2 (Utilities|Time|SetTimerbyEvent (EventDispatchers|CreateEvent :!event Custom|CountdownUpdate_Seconds) 1.0 true))
      (Variables|Default|SetCountdownUpdateSecondsTimerHandle _returnvalue_2))))

(event Custom|SetCountdownTimer
  (params (CountdownDuration :type Float%20%28double-precision%29))
  (Networking|SwitchHasAuthority
    (:Authority
      (bind _returnvalue (Utilities|Time|SetTimerbyEvent (EventDispatchers|CreateEvent :!event Custom|DoOnCountdownCompleted) CountdownDuration))
      (Variables|Transient|Console|SetCountdownTimerHandle _returnvalue))))

(event Custom|DoOnCountdownCompleted
  (Variables|Transient|Console|SetBasilOkra_State_Enum "NewEnumerator3"))

(event Custom|TimerRender_Update
  (Variables|Transient|Timer|SetCountdownRemaining (Utilities|Time|GetTimerRemainingTimebyHandle (Variables|Transient|Console|GetCountdownTimerHandle))))

(event Custom|CountdownUpdate_Seconds
  (Variables|Transient|Timer|SetCountdownSeconds (Utilities|Time|GetTimerRemainingTimebyHandle (Variables|Transient|Console|GetCountdownTimerHandle))))

(event Custom|BasilOkra_SetCountdownActive
  (Variables|Transient|Console|SetBasilOkra_State_Enum "NewEnumerator1")
  (Utilities|Time|ClearandInvalidateTimerbyHandle (Variables|Default|GetForceStartTimerHandle)))

(event Custom|PenalizeCountdown
  (bind _penalizedcountdownvalue (Console|GetPenalizedCountdownValue))
  (Utilities|FlowControl|Sequence
    (:then_0
      ;[1 1 1] Pause visual timer tick (we do this first to ensure no visual discrepancies when we modify the timer value)
      (Utilities|Time|PauseTimerbyHandle (Variables|Transient|Console|GetCountdownUpdateTimerHandle))
      (Utilities|Time|PauseTimerbyHandle (Variables|Default|GetCountdownUpdateSecondsTimerHandle)))
    (:then_1
      ;[1 1 1] Cache the time remaining on our countdown before clearing the timer (we cannot just update the value on timers, we have to clear and reset it)
      (Variables|Transient|Timer|SetCachedTimeRemaining (Utilities|Time|GetTimerRemainingTimebyHandle (Variables|Transient|Console|GetCountdownTimerHandle)))
      (Utilities|Time|ClearandInvalidateTimerbyHandle (Variables|Transient|Console|GetCountdownTimerHandle)))
    (:then_2
      ;[1 1 1] Check edge case of if penalizing our countdown would set it <= 0. If so, simply end the event.
      (if (<= _penalizedcountdownvalue 0)
        (Variables|Transient|Console|SetBasilOkra_State_Enum "NewEnumerator3")
        (else
          (CallFunction|SetCountdownTimer _penalizedcountdownvalue)
          (Variables|Transient|Timer|SetCountdownRemaining _penalizedcountdownvalue)
          (Utilities|Time|UnpauseTimerbyHandle (Variables|Transient|Console|GetCountdownUpdateTimerHandle))
          (Utilities|Time|UnpauseTimerbyHandle (Variables|Default|GetCountdownUpdateSecondsTimerHandle)))))))

(event Custom|DoOnMeltdownFailure
  (CallFunction|CleanupTimers)
  (for _array_element (Variables|Config|Console|GetVentInstanceArray)
    (Class|BCBasilOkraLootVent|TurnCollisionOn _array_element))
  (Utilities|Time|SetTimerbyEvent (EventDispatchers|CreateEvent :!event Custom|DoOnMeltdownDurationMet) (ScalableFloat|GetValueatLevel (Variables|Config|Console|GetMeltdownDuration))))

(event Custom|DoOnMeltdownDurationMet
  (for _array_element (Variables|Config|Console|GetVentInstanceArray)
    (Class|BCBasilOkraLootVent|TurnCollisionOff _array_element))
  (Rendering|Components|TextRender|SetText (Variables|Default|GetCountdown_TextRender) (Utilities|Text|MakeLiteralText "NSLOCTEXT(\"[6DF35C299CF81AABA04BBF2C5CF22E44]\", \"E0B08BAF49962399BD64EF978D9E86F0\", \"YOU TRIED. . .\")"))
  (Variables|Transient|Console|SetBasilOkra_State_Enum "NewEnumerator4"))

(fn CreateAndAssignConsoleMID)

(fn GetPenalizedCountdownValue
  (return (- (Variables|Transient|Timer|GetCachedTimeRemaining) (ScalableFloat|GetValueatLevel (Variables|Config|Console|GetCountdownPenalty)))))

(fn GetRandomInstructionManualLocation
  (params (NumLocationsToFetch :type Integer))
  (Utilities|Array|Shuffle (Variables|Config|Instructions|GetInstructionSpawnArray))
  (for _index (range NumLocationsToFetch)
    (Utilities|Array|Add (LocalVariables|GetLocalCachTransforms) (Transformation|GetWorldTransform (Variables|/WildEstate_BasilOkra/Gameplay/Actors/A_BasilOkra_Instructions_Spawner.A_BasilOkra_Instructions_Spawner_C|GetDefaultSceneRoot (Utilities|Array|Get(acopy) (Variables|Config|Instructions|GetInstructionSpawnArray) _index)))))
  (return (LocalVariables|GetLocalCachTransforms)))

(fn InitializeButtonActors
  (Networking|SwitchHasAuthority
    (:Authority
      ;[0 0 0 0.5] Spawn console button actors at the specified transforms
      (Default|SpawnButtonActor "/WildEstate_BasilOkra/Gameplay/Actors/BGA_BasilOkra_Button_1.BGA_BasilOkra_Button_1_C" (Transformation|GetWorldTransform (Variables|Default|GetButton1Transform)))
      (Default|SpawnButtonActor "/WildEstate_BasilOkra/Gameplay/Actors/BGA_BasilOkra_Button_2.BGA_BasilOkra_Button_2_C" (Transformation|GetWorldTransform (Variables|Default|GetButton2Transform)))
      (Default|SpawnButtonActor "/WildEstate_BasilOkra/Gameplay/Actors/BGA_BasilOkra_Button_3.BGA_BasilOkra_Button_3_C" (Transformation|GetWorldTransform (Variables|Default|GetButton3Transform))))))

(fn MarkGlobalScreenPing
  (Networking|SwitchHasAuthority
    (:Authority
      (for _array_element (Fortnite|GetAllFortPlayerControllers true true)
        (Utilities|IsValid _array_element
          (:Is%20Valid
            (Utilities|IsValid (Actor|GetComponentbyClass _array_element "/Script/FortniteGame.AthenaMarkerComponent")
              (:Is%20Valid
                (MarkerComponent|MarkActoronServer (Actor|GetComponentbyClass _array_element "/Script/FortniteGame.AthenaMarkerComponent") self false true)))))))))

(fn OnRep_BasilOkra_State_Enum
  (switch Utilities|FlowControl|Switch|SwitchonEnum_BasilOkra_States (Variables|Transient|Console|GetBasilOkra_State_Enum)
    (:NewEnumerator0
      (CallFunction|SetInteractState false)
      (CallFunction|TurnPostProcessOff))
    (:NewEnumerator1
      (Variables|Sound|SetbCountdownTimerIsActive? true)
      (CallFunction|SetCountdownTimer (ScalableFloat|GetValueatLevel (Variables|Config|Console|GetCountdownDuration)))
      (CallFunction|SetUpCountdownTextRender)
      (CallFunction|ActivateMeltdownVisuals)
      (CallFunction|ActivateMeltdownUI)
      (CallFunction|SpawnGlobalMarker)
      (CallFunction|SetInteractState true)
      (CallFunction|BroadcastGlobalEventActivated)
      (CallFunction|TurnPostProcessOn)
      ;[0.1208 0.04706 0.15 0.5] Audio
      ;;[0.15 0.08083 0.1004 0.5] Play POI Activated Sounds
      (Utilities|FlowControl|DoOnce false
        (:Completed
          (Components|Activation|Activate (Variables|/Script/Engine.AmbientSound|GetAudioComponent (Variables|Sound|GetPOI%20Activated%20Sound%20Fronts)) false)
          (Components|Activation|Activate (Variables|/Script/Engine.Actor|GetRootComponent (Variables|Sound|GetPOI%20Activated%20Sound%20Rears)) false))))
    (:NewEnumerator2
      (Variables|Sound|SetbPuzzleSucceeded? true)
      (Variables|Sound|SetbCountdownTimerIsActive? false)
      (CallFunction|DoonSequenceSuccess)
      (CallFunction|CleanupTimers)
      (CallFunction|DeactivateMeltdownVisuals)
      (CallFunction|DeactivateMeltdownUI)
      (CallFunction|SetInteractState false)
      (CallFunction|CleanupGlobalMarker)
      (CallFunction|TurnPostProcessOff)
      (CallFunction|BroadcastGlobalEventSuccess)
      ;;[0.3623 0.418 0.735] Play Succeed Stinger
      (Utilities|FlowControl|DoOnce false
        (:Completed
          (Variables|Sound|SetbCountdownTimerIsActive? false)
          (CallFunction|StopLoopingMusic 0.1)
          (CallFunction|PlayShutdownAudio)
          (CallFunction|PlayMusicSuccessStinger))))
    (:NewEnumerator3
      (Variables|Sound|SetbCountdownTimerIsActive? false)
      (CallFunction|DoonMeltdownFailure)
      (Class|BPBasilOkraVisualCollision|ActivateCollision (Variables|Config|GetVisualCollisionInstance))
      (CallFunction|StartCameraShake)
      (CallFunction|SetInteractState false true)
      (CallFunction|TurnPostProcessOff)
      ;;[0.3623 0.418 0.735] Play Time Exceeded Cadence
      (CallFunction|StopLoopingMusic 0.1)
      (CallFunction|PlayMeltdownBuildup)
      (CallFunction|PlayShutdownAudio)
      (CallFunction|PlayMusicFailStinger))
    (:NewEnumerator4
      (CallFunction|DeactivateMeltdownVisuals)
      (CallFunction|CleanupGlobalMarker)
      (CallFunction|TerminateCameraShake)
      (CallFunction|TurnPostProcessOff))))

(fn OnRep_bCountdownTimerIsActive?)

(fn OnRep_bPuzzleSucceeded?)

(fn OnRep_CachedInstructionLocation
  (Transformation|SetActorLocation (Variables|Transient|Instructions|GetInstructions) (Variables|Transient|Instructions|GetCachedInstructionLocation) false false))

(fn OnRep_CachedInstructionLocations
  (bind _instructionsarray (Variables|Transient|Instructions|GetInstructionsArray))
  (bind _output (Utilities|Array|Get(acopy) _instructionsarray))
  (for _array_element (Variables|Transient|Instructions|GetCachedInstructionLocations)
    (Utilities|IsValid _output
      (:Is%20Valid
        (Transformation|SetActorTransform _output _array_element false false)))))

(fn OnRep_CorrectButtonTagArray
  (for _array_element (Variables|Transient|Instructions|GetInstructionsArray)
    (Utilities|IsValid _array_element
      (:Is%20Valid
        (Variables|/WildEstate_BasilOkra/Gameplay/Actors/BGA_BasilOkra_Instructions.BGA_BasilOkra_Instructions_C|SetTag%20Sequence (Variables|Transient|Console|GetCorrectButtonTagArray) _array_element)))))

(fn OnRep_CountdownRemaining
  (Networking|CanExecuteCosmeticEvents
    (:True
      (Console|UpdateMPCTimeDigits))))

(fn OnRep_CountdownSeconds
  (Networking|CanExecuteCosmeticEvents
    (:True
      ;[0.04795 0.4926 0.7607 0.5] Ticking Sound For Each Second
      (Audio|SpawnSoundatLocation (Variables|Sound|GetTimer%20Tick%20Sound) (Transformation|GetWorldLocation (Variables|Default|GetCountdown_TextRender))))))

(fn OnRep_CurrentButtonTagArray
  (bind _currentbuttontagarray (Variables|Transient|Console|GetCurrentButtonTagArray))
  (Networking|CanExecuteCosmeticEvents
    (:True
      (if (Utilities|Array|IsNotEmpty _currentbuttontagarray)
        (CallFunction|UpdateSequenceVisualAll (Utilities|Array|Get(aref) _currentbuttontagarray (Utilities|Array|LastIndex _currentbuttontagarray)) (Utilities|Array|LastIndex _currentbuttontagarray))))))

(fn OnRep_InteractingPawn)

(fn OnRep_PostProcessAlpha
  (CallFunction|UpdatePostProcessAlpha))

(fn ResetSequenceVisualsToWarning
  (for _index (range 3)
    (CallFunction|SetSequenceVisual :SequenceIndex _index))
  (Console|SetAllWarningStateVisuals true))

(fn SetAllWarningStateVisuals
  (params (ActiveState :type Boolean))
  (for _index (range 3)
    (Console|SetWarningStateVisualsonConsole _index ActiveState)))

(fn SetSequenceVisual
  (params (ButtonPressed :type Gameplay%20Tag%20Structure) (SequenceIndex :type Integer))
  (Utilities|FlowControl|Sequence
    (:then_0
      (switch Utilities|FlowControl|Switch|SwitchonGameplayTag (Variables|GetButtonPressed)
        (:Default
          (LocalVariables|SetSelectedButton 0))
        (:Gameplay.Status.WildEstate.BasilOkra.Button.1
          (LocalVariables|SetSelectedButton 1))
        (:Gameplay.Status.WildEstate.BasilOkra.Button.2
          (LocalVariables|SetSelectedButton 2))
        (:Gameplay.Status.WildEstate.BasilOkra.Button.3
          (LocalVariables|SetSelectedButton 3))))
    (:then_1
      (Rendering|Material|SetScalarParameterValue "/WildEstateTech/TechArt/GlowDonut/Timer/MPC_GlowDonutState.MPC_GlowDonutState" (Utilities|Select :Index SequenceIndex :Option%200 "InputA" :Option%201 "InputB" :Option%202 "InputC") (Math|Conversions|ToFloat(Integer) (LocalVariables|GetSelectedButton)) :!class /Script/Engine.KismetMaterialLibrary)
      (Development|PrintText (Utilities|Text|FormatText "NSLOCTEXT(\"[6DF35C299CF81AABA04BBF2C5CF22E44]\", \"4802C0CA437493F85F7558A0C07515A8\", \"Button {x} set to selected index \\\"{y}\\\"\")" (LocalVariables|GetSelectedButton) SequenceIndex)))))

(fn SetWarningStateVisualsOnConsole
  (params (WarningIndex :type Integer) (Active :type Boolean))
  (bind _returnvalue (Utilities|Select :Index WarningIndex :Option%200 "ShowConsoleWarningA" :Option%201 "ShowConsoleWarningB" :Option%202 "ShowConsoleWarningC"))
  (Development|PrintString (Utilities|String|Append "CURRENT VISUAL OPTION !!! " (Utilities|String|ToString(Name) _returnvalue)))
  (Rendering|Material|SetScalarParameterValue "/WildEstateTech/TechArt/GlowDonut/Timer/MPC_GlowDonutState.MPC_GlowDonutState" _returnvalue (Math|Float|SelectFloat 1.0 0.0 Active) :!class /Script/Engine.KismetMaterialLibrary))

(fn SpawnButtonActor
  (params (ButtonClass :type BGA%20Basil%20Okra%20Button%20Base%20Class%20Reference) (SpawnTransform :type Transform))
  (bind _returnvalue (Game|SpawnActorfromClass ButtonClass SpawnTransform "AlwaysSpawn" "MultiplyWithRoot" self 0 0))
  (Utilities|Array|Add (Variables|Default|GetButtonActorArray) _returnvalue)
  (Transformation|AttachActorToComponent _returnvalue (Variables|Default|GetButtons) "None" "KeepWorld" "KeepWorld" "KeepWorld" true))

(fn UpdateMPCTimeDigits
  (bind _countdownremaining (Variables|Transient|Timer|GetCountdownRemaining))
  (bind _returnvalue (Math|Float|SafeDivide _countdownremaining 6e+01))
  (bind _returnvalue_2 (Math|Float|Floor _returnvalue))
  (bind _returnvalue_3 (* _returnvalue_2 60))
  (bind _returnvalue_4 (- _countdownremaining _returnvalue_3))
  (bind _returnvalue_5 (Math|Float|Floor _returnvalue_4))
  ;[0 0 0 0.5] Minutes
  (Rendering|Material|SetScalarParameterValue "/WildEstateTech/TechArt/GlowDonut/Timer/MPC_GlowDonutState.MPC_GlowDonutState" "Digit3" (Math|Conversions|ToFloat(Integer) _returnvalue_2) :!class /Script/Engine.KismetMaterialLibrary)
  ;[0 0 0 0.5] Seconds 1st digit
  (Rendering|Material|SetScalarParameterValue "/WildEstateTech/TechArt/GlowDonut/Timer/MPC_GlowDonutState.MPC_GlowDonutState" "Digit2" (Math|Conversions|ToFloat(Integer) (/ _returnvalue_5 10)) :!class /Script/Engine.KismetMaterialLibrary)
  ;[0 0 0 0.5] Seconds 2nd digit
  (Rendering|Material|SetScalarParameterValue "/WildEstateTech/TechArt/GlowDonut/Timer/MPC_GlowDonutState.MPC_GlowDonutState" "Digit1" (Math|Conversions|ToFloat(Integer) (Math|Integer|%25%28Integer%29 _returnvalue_5 10)) :!class /Script/Engine.KismetMaterialLibrary))

(fn UpdatePostProcessAlpha
  (Rendering|Material|SetScalarParameterValue "/WildEstateTech/TechArt/GlowDonut/Timer/MPC_GlowDonutState.MPC_GlowDonutState" "MMOPostProcessAlpha" (Variables|Transient|GetPostProcessAlpha) :!class /Script/Engine.KismetMaterialLibrary))

(fn ConstructionScript)
