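; Basil Okra instructions actor that passes the correct three-button gameplay tag sequence to a glow donut material.
; SetCorrectSequence stores the tags in Tag Sequence, whose OnRep maps each button tag to a 0-2 index.
; Each index is written to InstructionTargetIndex A, B or C in the shared MPC_GlowDonutState parameter collection.
; The actor destroys itself when its linked actor is destroyed.
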(event Custom|SetCorrectSequence
  (params (TagSequence :type Array%20of%20Gameplay%20Tag%20Structures))
  (Variables|Default|SetTag%20Sequence TagSequence))

(event OnLinkedActorDestroyed(FortLinkToActor)
  (params (DamageCauser :type Actor%20Object%20Reference))
  (Actor|DestroyActor self))

(fn OnRep_TagSequence
  (for (_array_element _array_index) (Variables|Default|GetTag%20Sequence)
    (CallFunction|SetMPCCorrectSequence _array_element _array_index)))

(fn SetMPCCorrectSequence
  (params (Selection :type Gameplay%20Tag%20Structure) (Index :type Integer))
  (Utilities|FlowControl|Sequence
    (:then_0
      (switch Utilities|FlowControl|Switch|SwitchonGameplayTag Selection
        (:Default
          (Development|PrintString "ERROR: Could not find Selection tag when setting button order in MPC"))
        (:Gameplay.Status.WildEstate.BasilOkra.Button.1
          (LocalVariables|SetSelectedButton 0))
        (:Gameplay.Status.WildEstate.BasilOkra.Button.2
          (LocalVariables|SetSelectedButton 1))
        (:Gameplay.Status.WildEstate.BasilOkra.Button.3
          (LocalVariables|SetSelectedButton 2))))
    (:then_1
      (Rendering|Material|SetScalarParameterValue "/WildEstateTech/TechArt/GlowDonut/Timer/MPC_GlowDonutState.MPC_GlowDonutState" (Utilities|Select :Index Index :Option%200 "InstructionTargetIndexA" :Option%201 "InstructionTargetIndexB" :Option%202 "InstructionTargetIndexC") (Math|Conversions|ToFloat(Integer) (LocalVariables|GetSelectedButton)) :!class /Script/Engine.KismetMaterialLibrary))))

(fn ConstructionScript)
