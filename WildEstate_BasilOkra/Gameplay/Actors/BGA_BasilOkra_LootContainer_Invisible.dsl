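; Invisible BasilOkra loot container actor that spawns loot at its own location, using a loot group chosen by the current playlist.
; Custom event SpawnLootByTierGroup feeds the result of ParseLootGroupByPlaylist into SpawnLootfromTableUsingItemEntry.
; That function looks up each Athena playlist context tag in LootGroupByPlaylistMap, falling back to DefaultLootGroup without a non-empty match.
; BeginPlay, overlap and Tick are empty, and nothing here calls the spawn event.
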(event EventBeginPlay)

(event Collision|EventActorBeginOverlap
  (params (OtherActor :type Actor%20Object%20Reference)))

(event EventTick
  (params (DeltaSeconds :type Float%20%28single-precision%29)))

(event Custom|SpawnLootByTierGroup
  (Class|AthenaFunctionLibrary|SpawnLootfromTableUsingItemEntry (CallFunction|ParseLootGroupbyPlaylist) 1 -1 (Transformation|GetActorLocation self) "0.000000,0.000000,1.000000" 0 true true))

(fn ParseLootGroupByPlaylist
  (bind _lootgroupbyplaylistmap (Variables|Default|GetLootGroupByPlaylistMap))
  (bind _array_element (Utilities|Array|ForEachLoopwithBreak (GameplayTags|BreakGameplayTagContainer (Class|FortGameStateAthena|GetAthenaPlaylistContextTags (Fortnite|Athena|GameState|GetGameStateAthena))))
    (:LoopBody
      (if (.ReturnValue (Utilities|Map|Find _lootgroupbyplaylistmap _array_element))
        (Variables|Default|SetLootGroupFound true)
        (Variables|Default|SetLootGroupName (Utilities|Map|Find _lootgroupbyplaylistmap _array_element))))
    (:Completed
      _))
  (if (and (Variables|Default|GetLootGroupFound) (!= (Variables|Default|GetLootGroupName) ""))
    (return (Variables|Default|GetLootGroupName))
    (else
      (return (Variables|Default|GetDefaultLootGroup)))))

(fn ConstructionScript)
