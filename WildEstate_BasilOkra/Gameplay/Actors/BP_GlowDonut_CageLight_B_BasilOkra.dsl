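; Cage light Blueprint actor for the BasilOkra Wild Estate, with an alarm-style glow that switches on and off.
; ToggleVisuals activates or deactivates the radial alarm effect component to match its Enabled flag.
; It also sets the static mesh's dynamic material emissive tint to white when enabled and black when disabled.
; BeginPlay and construction only call the parent, and Tick and overlap events are empty, so nothing here triggers the toggle.
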(event EventBeginPlay
  (|CallParentFunction))

(event Collision|EventActorBeginOverlap
  (params (OtherActor :type Actor%20Object%20Reference)))

(event EventTick
  (params (DeltaSeconds :type Float%20%28single-precision%29)))

(fn ToggleVisuals
  (params (Enabled :type Boolean))
  (Components|Activation|SetActive (Variables|Default|GetNS_BasilOkra_Alarm_Radial) Enabled Enabled)
  (if Enabled
    (Rendering|Material|SetVectorParameterValue (Class|BTechArtGlobalLibrary|GetDynamicMaterialInstance (Variables|Debug|GetStaticMeshComponent)) "SFX_Emissive_Tint(Expand)_1" "(R=1.000000,G=1.000000,B=1.000000,A=1.000000)" :!class /Script/Engine.MaterialInstanceDynamic)
    (else
      (Rendering|Material|SetVectorParameterValue (Class|BTechArtGlobalLibrary|GetDynamicMaterialInstance (Variables|Debug|GetStaticMeshComponent)) "SFX_Emissive_Tint(Expand)_1" "(R=0.000000,G=0.000000,B=0.000000,A=1.000000)" :!class /Script/Engine.MaterialInstanceDynamic))))

(fn ConstructionScript
  (|CallParentFunction))
