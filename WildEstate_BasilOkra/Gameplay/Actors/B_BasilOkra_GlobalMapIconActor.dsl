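; Map icon actor for the BasilOkra Wild Estate gameplay that carries a Fort minimap component to show an icon on the map.
; On BeginPlay it flushes net dormancy so the actor replicates to clients.
; SetMapIcon runs only with authority, setting the minimap component's icon and flushing dormancy again to push the change to clients.
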(event EventBeginPlay
  ;[0.007568 0.06955 0.07812] BeginPlay: Replicate to the client.
  (|CallParentFunction)
  (Networking|FlushNetDormancy))

(fn SetMapIcon
  (params (MiniMapIcon :type Object%20Reference))
  ;[0.02466 0.1126 0.1198] SetMapIcon: Update the map icon.
  (Networking|SwitchHasAuthority
    (:Authority
      (MiniMapData|SetMiniMapIcon (Variables|Default|GetFortMiniMapComponent) MiniMapIcon)
      (Networking|FlushNetDormancy))))

(fn ConstructionScript)
