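; Game state component that drives the Basil Okra match objective by starting the countdown on the world's console actor once.
; On authority it can pick a random safe zone phase at playlist load and activates when that phase and hotfixed state arrive, otherwise deferring to parent logic.
; It forces activation early if the console falls outside the next safe zone.
; Replication of the cached console optionally pings it for nearby unplaced local players.
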(event EventBeginPlay
  ;[0 0 0 0.5] Initialize random activation safe zone phase
  (|CallParentFunction)
  (Utilities|SwitchHasAuthority
    (:Authority
      (if (ScalableFloat|GetValueasBool (Variables|Default|GetbUseRandomizedSafeZonePhase))
        (Class|FortAsyncActionCurrentPlaylistReady|PlaylistDataReadyAsync
          (:OnReady
            (Variables|Default|SetRandomizedActivationPhase (Math|Random|RandomIntegerinRange (ScalableFloat|GetValueasInteger (Variables|Default|GetRandomizedActivationPhaseMin)) (ScalableFloat|GetValueasInteger (Variables|Default|GetRandomizedActivationPhaseMax))))))))))

(event EventActivateObjective
  (bind _consoleactor (CallFunction|TryGetBasilOkraConsoleActor))
  (|CallParentFunction)
  (if (not (Variables|Default|GetbHasBeenActivated))
    (Utilities|IsValid _consoleactor
      (:Is%20Valid
        (Class|BGABasilOkraConsole|BasilOkraSetCountdownActive _consoleactor)
        (Variables|Default|SetCachedConsoleActor _consoleactor)
        (Variables|Default|SetbHasBeenActivated true))
      (:Is%20Not%20Valid
        (Fortnite|FortShippingLog self "Basil Okra console actor is invalid. MMO will not be activated" true)))))

(event Custom|PingAllPlayersInProximity
  (Networking|CanExecuteCosmeticEvents
    (:True
      (for _array_element (Fortnite|GetLocalFortPlayerControllers true)
        (Utilities|IsValid _array_element
          (:Is%20Valid
            (Utilities|IsValid (Pawn|GetViewTarget _array_element)
              (:Is%20Valid
                (bind _asfort_player_state_athena (Utilities|Casting|CastToFortPlayerStateAthena (Variables|/Script/Engine.Controller|GetPlayerState _array_element)))
                (if (<= (Variables|/Script/FortniteGame.FortPlayerStateAthena|GetPlace _asfort_player_state_athena) 0)
                  (Utilities|IsValid (Actor|GetComponentbyClass _array_element "/Script/FortniteGame.AthenaMarkerComponent")
                    (:Is%20Valid
                      (if (<= (Math|Vector|Distance2D(Vector) (Transformation|GetActorLocation (Variables|Default|GetCachedConsoleActor)) (Transformation|GetActorLocation (Pawn|GetViewTarget _array_element))) (ScalableFloat|GetValueatLevel (Variables|Default|GetPingProximityRadii)))
                        (MarkableActor|SetCanBeMarked (Variables|Default|GetCachedConsoleActor) true) ; Ensure the actor can be marked in case its state variable hasn't replicated yet.
                        (Fortnite|RegisterSoundIndicatorPlacementContext _array_element (Variables|Default|GetCachedConsoleActor) "0, 0, 0" false)))))))))))))

(fn HandleBasilOkraActivation
  (CallFunction|HandleDelayedActivation)
  (CallFunction|UnbindSafeZoneDelegates))

(fn HandleSafeZoneUpdated
  (params (PhaseIndex :type Integer) (SafeZoneState :type EFortSafeZoneState%20Enum))
  (if (Networking|IsServer)
    (Networking|SwitchHasAuthority
      (:Authority
        (if (CallFunction|ShouldForceSpawn)
          ;[0 0 0 0.5] Console actor is outside of next safe zone. Force activation for this phase
          (Fortnite|FortShippingLog self (Utilities|String|Append "Basil Okra MMO outside of next storm phase. Forcing activation for phase: " (Utilities|String|ToString(Integer) (Variables|GetPhaseIndex))))
          (CallFunction|HandleBasilOkraActivation)
          (else
            ;[0 0 0 0.5] Override parent functionality only if we want to randomize the safe zone activation phase
            (if (ScalableFloat|GetValueasBool (Variables|Default|GetbUseRandomizedSafeZonePhase))
              (Fortnite|FortShippingLog self (Utilities|String|Append "HandleSafeZoneUpdated: Phase Index is " (Utilities|String|ToString(Integer) (Variables|GetPhaseIndex)) " ; SafeZoneState is " (Utilities|String|EnumtoString (Variables|GetSafeZoneState))))
              ;[0 0 0 0.5] Check if the phase and state match our configured values
              (if (and (== (Variables|Default|GetRandomizedActivationPhase) (Variables|GetPhaseIndex)) (== (Math|Conversions|ToByte(Integer) (ScalableFloat|GetValueasInteger (Variables|Hotfix|GetHotfix_ActivateSafeZoneState))) (Variables|GetSafeZoneState)))
                (CallFunction|HandleBasilOkraActivation))
              (else
                (|CallParentFunction (Variables|GetPhaseIndex) (Variables|GetSafeZoneState))))))))))

(fn OnRep_CachedConsoleActor
  (if (ScalableFloat|GetValueasBool (Variables|Default|GetShouldPingOnProximity))
    (CallFunction|PingAllPlayersinProximity)))

(fn ShouldForceSpawn
  (bind _consoleactor (CallFunction|TryGetBasilOkraConsoleActor))
  (if (ScalableFloat|GetValueasBool (Variables|Default|GetbForceSpawnIfOutsideNextSafeZone))
    (Utilities|IsValid _consoleactor
      (:Is%20Valid
        (bind _safezoneindicator (Variables|Default|GetSafeZoneIndicator)
          (:then
            (return (not (SafeZone|IsinNextSafeZone _safezoneindicator (Transformation|GetActorLocation _consoleactor)))))
          (:else
            (return false))))
      (:Is%20Not%20Valid
        (Fortnite|FortShippingLog self "Basil Okra console actor is invalid. MMO will not be activated" true)))
    (else
      (return false))))

(fn TryGetBasilOkraConsoleActor
  (Variables|Default|GetCachedConsoleActor
    (:then
      ;[0 0 0 0.5] Return cached actor if valid
      (return (Variables|Default|GetCachedConsoleActor)))
    (:else
      (bind _returnvalue (Actor|GetActorOfClass "/WildEstate_BasilOkra/Gameplay/Actors/BGA_BasilOkra_Console.BGA_BasilOkra_Console_C"))
      (Utilities|IsValid _returnvalue
        (:Is%20Valid
          (return _returnvalue))
        (:Is%20Not%20Valid
          (Fortnite|FortShippingLog self "Basil Okra console actor not found. MMO will not be activated" true))))))
